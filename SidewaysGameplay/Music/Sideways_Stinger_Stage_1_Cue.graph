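# Sound cue for the Sideways stage 1 music stinger, playing a 120 BPM combat intro track.
# A fixed one-second delay feeds a single wave player, with no random variation.
# It plays through the encounter music attenuation and the POI music sound class at high priority, with a slightly reduced volume.
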SoundCue /SidewaysGameplay/Music/Sideways_Stinger_Stage_1_Cue
root=Delay_0
nodes [
 {id=Delay_0, children [WavePlayer_0], properties {delayMax=1, delayMin=1}}
 {id=WavePlayer_0, properties {soundWaveAssetPtr=&/S18_Sideways_Combat_Intro_120BPM}}
]
properties {
 attenuationSettings=&/Sideways_Encounter_Music_Att
 maxDistance=3000
 priority=100
 soundClassObject=/Game/Sounds/FortSoundClassesAndModes/Fort_Music_POI_PSM
 volumeMultiplier=0.9524
}
