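# Sound cue for the Sideways encounter's stage 2 music stinger, playing the Season 18 combat stage 2 track at 120 BPM.
# A fixed half-second delay node feeds a single wave player.
# It uses the Sideways encounter music attenuation, a POI music sound class and a high priority.
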SoundCue /SidewaysGameplay/Music/Sideways_Stinger_Stage_2_Cue
root=Delay_0
nodes [
 {id=Delay_0, children [WavePlayer_1], properties {delayMax=0.5, delayMin=0.5}}
 {id=WavePlayer_1, properties {soundWaveAssetPtr=&/S18_Sideways_Combat_Stage_2_120BPM}}
]
properties {
 attenuationSettings=&/Sideways_Encounter_Music_Att
 maxDistance=3000
 priority=100
 soundClassObject=/Game/Sounds/FortSoundClassesAndModes/Fort_Music_POI_PSM
 volumeMultiplier=0.9524
}
