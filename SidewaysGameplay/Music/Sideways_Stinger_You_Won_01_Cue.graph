SoundCue /SidewaysGameplay/Music/Sideways_Stinger_You_Won_01_Cue
root=Delay_0
nodes [
 {id=Delay_0, children [WavePlayer_1], properties {delayMax=1.25, delayMin=1.25}}
 {id=WavePlayer_1, properties {soundWaveAssetPtr=&/S18_Sideways_Combat_Victory_120BPM}}
]
properties {
 attenuationSettings=&/Sideways_Encounter_Music_Att
 maxDistance=3000
 priority=100
 soundClassObject=/Game/Sounds/FortSoundClassesAndModes/Fort_Music_POI_PSM
 volumeMultiplier=1.1905
 concurrencySet [&/Sideways_Music_Victory_Conc]
}
